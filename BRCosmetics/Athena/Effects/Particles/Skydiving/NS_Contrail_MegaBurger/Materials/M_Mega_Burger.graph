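# Unlit masked material for Niagara mesh particles in the MegaBurger skydiving contrail.
# Emissive blends the burger texture with a particle-color-tinted version via UseParticleColor, then passes through the shared MF_FX_MAEmissiveAdjustments function.
# A normal map parameter, flat by default, feeds Normal, while OpacityMask is unconnected.
# A world-position sine wave masked by vertex red is built but not wired to WorldPositionOffset, and it has no time input.
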Material /BRCosmetics/Athena/Effects/Particles/Skydiving/NS_Contrail_MegaBurger/Materials/M_Mega_Burger
graph {
 properties {
  bAutomaticallySetUsageInEditor=false
  bUseMaterialAttributes=true
  bUsedWithNiagaraMeshParticles=true
  blendMode=BLEND_Masked
  shadingModel=MSM_Unlit
  usedShadingModels=Unlit
 }
 nodes [
  n1 = ScalarParameter() {defaultValue=2, group=WorldPositionOffset, parameterName=WPO_Speed_R, sortPriority=1}
  n2 = ScalarParameter() {defaultValue=2, group=WorldPositionOffset, parameterName=WPO_Speed_G, sortPriority=1}
  n3 = ScalarParameter() {defaultValue=2, group=WorldPositionOffset, parameterName=WPO_Speed_B, sortPriority=1}
  n4 = WorldPosition()
  n5 = MaterialFunctionCall(
   X=ScalarParameter() {defaultValue=0.5, group=WorldPositionOffset, parameterName=WPO_WaveScale_R, sortPriority=3}
   Y=ScalarParameter() {defaultValue=0.5, group=WorldPositionOffset, parameterName=WPO_WaveScale_G, sortPriority=3}
   Z=ScalarParameter() {defaultValue=0.5, group=WorldPositionOffset, parameterName=WPO_WaveScale_B, sortPriority=3}
  ) {materialFunction=&MakeFloat3}
  n6 = Sine(None=Add(
    A=MaterialFunctionCall(X=@n1, Y=@n2, Z=@n3) {materialFunction=&MakeFloat3}.Result
    B=Divide(A=@n4.XYZ, B=@n5.Result)
   ))
  n7 = MaterialFunctionCall(
   X=ScalarParameter() {defaultValue=1, group=WorldPositionOffset, parameterName=WPO_LargeWave_R, sortPriority=2}
   Y=ScalarParameter() {defaultValue=1, group=WorldPositionOffset, parameterName=WPO_LargeWave_G, sortPriority=2}
   Z=ScalarParameter() {defaultValue=1, group=WorldPositionOffset, parameterName=WPO_LargeWave_B, sortPriority=2}
  ) {materialFunction=&MakeFloat3}
  ComponentMask(None=Multiply(A=Multiply(A=@n6, B=@n7.Result), B=VertexColor().R)) {b=true, g=true, r=true}
  n8 = TextureSample() {texture=/BRCosmetics/Athena/Effects/Particles/Skydiving/NS_Contrail_MegaBurger/Textures/T_M_Mega_Burger_D}
  n9 = Multiply(A=@n8.RGB, B=ParticleColor().RGB)
  n10 = ScalarParameter() {parameterName=UseParticleColor}
  n11 = TextureSampleParameter2D() {
   group=Normal
   parameterName=Normals
   samplerSource=SSM_Wrap_WorldGroupSettings
   samplerType=SAMPLERTYPE_Normal
   texture=/Engine/EngineMaterials/FlatNormal
  }
  MaterialFunctionCall!(In=MakeMaterialAttributes!(
    BaseColor
    Metallic
    Specular
    Roughness
    Anisotropy
    EmissiveColor=LinearInterpolate(A=@n8.RGB, B=@n9, Alpha=@n10)
    Opacity
    OpacityMask
    Normal=@n11.RGB
    Tangent
    WorldPositionOffset
    SubsurfaceColor
    ClearCoat
    ClearCoatRoughness
    AmbientOcclusion
    Refraction
    CustomizedUVs_0
    CustomizedUVs_1
    CustomizedUVs_2
    CustomizedUVs_3
    CustomizedUVs_4
    CustomizedUVs_5
    CustomizedUVs_6
    CustomizedUVs_7
    PixelDepthOffset
    ShadingModel
    Displacement
   )) {materialFunction=/Game/Effects/Niagara/MasterMaterials/Functions/MF_FX_MAEmissiveAdjustments}
 ]
}
