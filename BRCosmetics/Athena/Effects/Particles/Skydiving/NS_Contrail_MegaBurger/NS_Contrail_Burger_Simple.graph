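# Niagara skydiving contrail for the MegaBurger cosmetic, trailing pendulum-driven chains from skeletal mesh bones or sockets.
# NE_Empty partitions particles into chains, simulates them with spring-driven pendulum constraints, owner-velocity drag and curl noise, and draws them as sprites and ribbons.
# Chain shape adjusts for front-end display, and ribbon width scales with dive alpha.
# The disabled NE_BurgerStack emitter would read those chains to stick cycling mesh particles along them.
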NiagaraSystem /BRCosmetics/Athena/Effects/Particles/Skydiving/NS_Contrail_MegaBurger/NS_Contrail_Burger_Simple
dependencies {
 usedDataInterfaces [
  &NiagaraDataInterfaceCurve
  &NiagaraDataInterfaceParticleRead
  &NiagaraDataInterfaceSkeletalMesh
 ]
 usedDynamicInputs [/Niagara/DynamicInputs/Lerp/SmoothLerpOverTime_Float]
 usedModules [
  &self:SystemSpawnScript.SystemScriptSource.SystemScriptGraph.NiagaraNodeAssignment_1.SetVariables_#1
  &self:SystemSpawnScript.SystemScriptSource.SystemScriptGraph.NiagaraNodeAssignment_0.SetVariables_#2
  &self:SystemSpawnScript.SystemScriptSource.SystemScriptGraph.NiagaraNodeAssignment_4.SetVariables_#3
  &self:SystemSpawnScript.SystemScriptSource.SystemScriptGraph.NiagaraNodeAssignment_3.SetVariables_#4
  &self:SystemSpawnScript.SystemScriptSource.SystemScriptGraph.NiagaraNodeAssignment_2.SetVariables_#5
  &SystemState
 ]
}
emitters {
 NE_BurgerStack {
  bEnabled=false
  emitterUpdateScript [
   &EmitterState(
    Inactive Response=::NewEnumerator1
    Life Cycle Mode=::NewEnumerator1
    Loop Behavior=::NewEnumerator1
    Loop Duration Mode=::NewEnumerator1
    UseLoopDelay=false
   )
   &SpawnBurst_Instantaneous(
    Spawn Count=&Multiply_Int(A=$System.ChainNumber, B=$System.LinkPerChain)
    Use Spawn Probability=false
    UseLoopCountLimit=false
   )
  ]
  particleSpawnScript [
   &InitializeParticle(
    Color Mode=::NewEnumerator0
    Lifetime=5
    Mesh Scale Mode=::NewEnumerator4
    Sprite Size Mode=::NewEnumerator3
    Uniform Sprite Size=100
    UsePositionOffset=false
    Write Material Random=true
   )
   &PartitionParticles(
    Number Of Partitions=$System.ChainNumber
    Particles Per Partition:NiagaraInt32=$System.LinkPerChain
    Partition Mode:ENiagara_PartitionMode_v2=::NewEnumerator1
   )
   &SkeletalMeshLocation(
    Bone / Socket Index=$Particles.PartitionParticles.Partition
    Bone Sampling Mode=::NewEnumerator4
    MaskSpawnGroup=false
    Sampled Position Offset=&Lerp_Vector(A=<0 0 0>, B=<0 0 100>, Alpha=$Particles.NormalizedChainPosition)
    Sampled Position Offset Coordinate Space:ENiagaraSkelMeshTransforms=::NewEnumerator1
    Skeletal Mesh=$System.Skeletal Mesh
   )
  ]
  particleUpdateScript [
   &ParticleState(Kill Particles When Lifetime Has Elapsed=false)
   /SetVariables(Particles.Scale=&Multiply_Vector(A=$Particles.Initial.Scale, B=<1 1 1>)) {isSetParameters=true}
   &/StickMeshtoRibbon(
    Particle Attribute Reader=DataInterface() {propertyValues {EmitterBinding {bindingMode=Other, emitterName=NE_Empty}}}
    ParticlePerRibbon:NiagaraInt32=$System.ParticlePerRibbon
    Partition:NiagaraInt32=$Particles.PartitionParticles.Partition
    RibbonSpace=2
   )
   &InterpolateOverTime(
    Define Initial Value=false
    Target 2D Value:Vector2f=&Multiply_Vector2D_ByFloat(Float=1, Vector2D=<1 1>)
    Variable Type=::NewEnumerator1
   )
   /SetVariables(Particles.MeshIndex=&ModuloInt(A=$Particles.PartitionParticles.Index In Partition, B=11)) {isSetParameters=true}
   &SolveForcesAndVelocity(Clamp Velocity=false, Limit Acceleration=false)
   &CalculateAccurateVelocity()
  ]
  rendererClasses [&NiagaraMeshRendererProperties]
  renderers [{rendererClass=&NiagaraMeshRendererProperties, rendererIndex=0}]
 }
 NE_Empty {
  emitterUpdateScript [
   &EmitterState(
    Inactive Response=::NewEnumerator1
    Life Cycle Mode=::NewEnumerator1
    Loop Behavior=::NewEnumerator1
    Loop Duration Mode=::NewEnumerator1
    UseLoopDelay=false
   ) {module=EmitterLifeCycle}
   &SpawnBurst_Instantaneous(
    Spawn Count=&Multiply_Int(A=$System.ChainNumber, B=$System.ParticlePerRibbon)
    Use Spawn Probability=false
    UseLoopCountLimit=false
   )
  ]
  particleSpawnScript [
   &InitializeParticle(Sprite Size Max=<20 20>, Sprite Size Mode=::NewEnumerator2, UsePositionOffset=false, Write Material Random=true)
   &PartitionParticles(
    Number Of Partitions=$System.ChainNumber
    Particles Per Partition:NiagaraInt32=$System.ParticlePerRibbon
    Partition Mode:ENiagara_PartitionMode=::NewEnumerator1
   )
   /SetVariables(Particles.NormalizedChainPosition:NiagaraFloat=$Particles.PartitionParticles.Normalized Index In Partition) {isSetParameters=true}
   &SkeletalMeshLocation(
    Bone / Socket Index=$Particles.PartitionParticles.Partition
    Bone Sampling Mode=::NewEnumerator4
    MaskSpawnGroup=false
    Sampled Position Offset=&Lerp_Vector(A=<0 0 0>, B=<0 0 150>, Alpha=$Particles.NormalizedChainPosition)
    Sampled Position Offset Coordinate Space:ENiagaraSkelMeshTransforms=::NewEnumerator1
    Skeletal Mesh=$System.Skeletal Mesh
   )
  ]
  particleUpdateScript [
   &ParticleState(Kill Particles When Lifetime Has Elapsed=false)
   &SkeletalMeshLocation(
    Bone / Socket Index=$Particles.PartitionParticles.Partition
    Bone Sampling Mode=::NewEnumerator4
    MaskSpawnGroup=false
    Position Sampling=::NewEnumerator2
    Sampled Position Offset=&Lerp_Vector(A=<10 -3 0>, B=<-10 3 0>, Alpha=&MakeFloatFromInt(Int=$Particles.PartitionParticles.Partition))
    Skeletal Mesh=$System.Skeletal Mesh
   ) {module=SkeletalMeshLocation001}
   &PendulumSetup(
    Module.Enable Spring Driver=true
    Pendulum Length:NiagaraFloat=&Lerp_Float(
     A=0
     B=$System.ChainLength
     Alpha=&FloatFromCurve(
      FloatCurve=DataInterface() {propertyValues {Curve {keys [<at 0 0>, {arriveTangent=-0.0703, interpMode=RCIM_Cubic, leaveTangent=-0.0703, tangentMode=RCTM_User, time=1, value=1}]}}}
      CurveIndex=$Particles.NormalizedChainPosition
      Scale Curve=&FloatFromCurve(
       FloatCurve=DataInterface() {propertyValues {Curve {keys [
           {arriveTangent=0.6092, interpMode=RCIM_Cubic, leaveTangent=0.6092, tangentMode=RCTM_User, time=0, value=0}
           {tangentMode=RCTM_User, time=1, value=1}
          ]}}}
       CurveIndex=$Emitter.Age
      )
     )
    )
    Pendulum Pivot:Vector3f=&AddVectorToPosition(
     Position=&AddVectorToPosition(Position=$Particles.SkeletalMeshLocation001.SampledPosition, Vector=<0 0 0>)
     Vector=&TransformVector(Vector=<0 0 0>, Source Space=::World, Destination Space=::Simulation)
    )
    Pendulum Pivot Offset Coordinate Space:ENiagaraCoordinateSpace=::Local
    Pendulum Rest Axis:Vector3f=&TransformVector(
     Vector=&Add_Vector(
      A=&Lerp_Vector(
       A=&Add_Vector(
        A=&Multiply_VectorByFloat(
         Vector=&NormalizeVector(Vector To Normalize=$Engine.Owner.Velocity, Safe Normalize=true, Safe Fallback Threshold=0, Safe Fallback Vector=<0 0 0.001>)
         Float=-0.5
        )
        B=&TransformVector(
         Vector=&RotateVectorByQuaternion(
          Vector To Rotate=&Multiply_VectorByFloat(
           Vector=<0.5 1 0>
           Float=&Lerp_Float(A=0.1, B=-0.1, Alpha=&MakeFloatFromInt(Int=$Particles.PartitionParticles.Partition))
          )
          Quaternion=$Particles.SkeletalMeshLocation001.SampledRotation
         )
         Destination Space=::World
        )
       )
       B=&Lerp_Vector(
        A=&Multiply_VectorByFloat(
         Vector=&NormalizeVector(Vector To Normalize=$Engine.Owner.Velocity, Safe Normalize=true, Safe Fallback Threshold=0, Safe Fallback Vector=<0 0 0.001>)
         Float=-1
        )
        B=<0 0 0.1>
        Alpha=0.5
       )
       Alpha=&Lerp_Float(
        A=&FloatFromCurve(
         FloatCurve=DataInterface() {propertyValues {Curve {keys [
             {arriveTangent=1.8207, interpMode=RCIM_Cubic, leaveTangent=1.8207, tangentMode=RCTM_User, time=0, value=0}
             {arriveTangent=-0.078, interpMode=RCIM_Cubic, leaveTangent=-0.078, tangentMode=RCTM_User, time=1, value=1}
            ]}}}
         CurveIndex=$Particles.NormalizedChainPosition
        )
        B=1
        Alpha=&FN_FrontEndFloatModifier(Gameplay=0, Front End=1)
       )
      )
      B=&TransformVector(
       Vector=&FN_FrontEndVector3Modifier(Gameplay=<0 0 0>, Front End=&Lerp_Vector(A=<0 -0.03 0>, B=<0 -0.15 0>, Alpha=$System.SmoothDiveAlpha))
       Destination Space=::World
      )
     )
     Source Space=::World
     Destination Space=::Simulation
    )
    "Spring Driven Constraint (Tightness)"=25
   )
   &LinearForce(
    Coordinate Space:ENiagaraCoordinateSpace=::World
    Force=&Multiply_VectorByFloat(Vector=$Engine.Owner.Velocity, Float=-5)
   )
   &CurlNoiseForce(
    Bias Noise=false
    Noise Frequency=10
    Noise Strength=&FloatFromCurve(
     FloatCurve=DataInterface() {propertyValues {Curve {keys [{tangentMode=RCTM_User, time=0, value=0}, {tangentMode=RCTM_User, time=1, value=1}]}}}
     CurveIndex=$Particles.NormalizedChainPosition
     Scale Curve=1000
    )
    Pan Noise=true
    Pan Noise Field=<0 0 2.2224>
   )
   &SolveForcesAndVelocity(Clamp Velocity=false, Limit Acceleration=false)
   &PendulumConstraint()
   &CalculateAccurateVelocity()
   /SetVariables(Particles.RibbonWidth=&FloatFromCurve(
     FloatCurve=DataInterface() {propertyValues {Curve {keys [
         {arriveTangent=8.0134, interpMode=RCIM_Cubic, leaveTangent=8.0134, tangentMode=RCTM_User, time=0, value=1}
         {arriveTangent=0.3471, interpMode=RCIM_Cubic, leaveTangent=0.3471, tangentMode=RCTM_User, time=0.1, value=1.5}
         {arriveTangent=-1.9168, interpMode=RCIM_Cubic, leaveTangent=-1.9168, tangentMode=RCTM_User, time=0.4, value=1}
         {interpMode=RCIM_Cubic, tangentMode=RCTM_Break, time=0.8, value=0.5}
         {arriveTangent=-0.7173, interpMode=RCIM_Cubic, leaveTangent=-0.7173, tangentMode=RCTM_User, time=1, value=0.3}
        ]}}}
     CurveIndex=$Particles.NormalizedChainPosition
     Scale Curve=&Lerp_Float(A=15, B=25, Alpha=$System.SmoothDiveAlpha)
    )) {isSetParameters=true}
  ]
  rendererClasses [&NiagaraSpriteRendererProperties, &NiagaraRibbonRendererProperties]
  renderers [
   {rendererClass=&NiagaraSpriteRendererProperties}
   {rendererClass=&NiagaraRibbonRendererProperties, rendererIndex=1}
  ]
 }
}
summary {
 systemName=NS_Contrail_Burger_Simple
 emitters [
  {emitterName=NE_Empty, rendererClasses [&NiagaraSpriteRendererProperties, &NiagaraRibbonRendererProperties]}
  {bEnabled=false, emitterName=NE_BurgerStack, rendererClasses [&NiagaraMeshRendererProperties]}
 ]
 userVariables [
  User.bIsFrontEnd {type=NiagaraBool}
  User.bIsFrontEndPreview {type=NiagaraBool, defaultValue=true}
  User.Debug {type=NiagaraBool, defaultValue=true}
  User.ParaGlideLeanAlpha {type=NiagaraFloat, defaultValue=0, description=""}
 ]
}
